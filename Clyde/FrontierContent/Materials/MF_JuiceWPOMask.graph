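# Material function that adjusts a world position offset vertically around a "Slurpy" sphere, likely for a juice effect.
# A SphereMask compares Position, absolute world position by default, with SlurpyCenter and SlurpyRadius from FortniteMaterialParameters, softened by SlurpyGradient.
# The inverted mask scales WPO_Amount, downward by default, and adds it to incoming Z, so the shift applies outside the sphere.
# Output Y reuses the incoming X component, discarding the original Y.
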MaterialFunction /Clyde/FrontierContent/Materials/MF_JuiceWPOMask
graph {nodes [
  n1 = MaterialFunctionCall(Color=FunctionInput() {inputName=WPO}) {materialFunction=&SplitComponents}
  n2 = FunctionInput(Preview=WorldPosition().XYZ) {bUsePreviewValueAsDefault=true, description=Defaults to absolute world Position, inputName=Position, previewValue=<-200 0 0 1>}
  n3 = ComponentMask(None=CollectionParameter() {collection=&FortniteMaterialParameters, parameterName=SlurpyCenter}) {b=true, g=true, r=true}
  n4 = ComponentMask(None=CollectionParameter() {collection=&FortniteMaterialParameters, parameterName=SlurpyRadius}) {b=true, g=true, r=true}
  n5 = ScalarParameter() {defaultValue=0.94, parameterName=SlurpyGradient}
  n6 = FunctionInput() {bUsePreviewValueAsDefault=true, inputName=WPO_Amount, inputType=FunctionInput_Scalar, previewValue=<-200 0 0 1>}
  n7 = Multiply(A=OneMinus(None=SphereMask(A=@n2, B=@n3, Radius=@n4, Hardness=@n5)), B=@n6)
  FunctionOutput(None=MaterialFunctionCall(X=@n1.R, Y=@n1.R, Z=Add(A=@n1.B, B=@n7)) {materialFunction=&MakeFloat3}.Result)
 ]}
